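# Physics asset for the Bani backpack cosmetic mesh, defining collision bodies and joints for its left and right ears.
# A single capsule covers the pack, and each ear is a three-capsule chain along its dyn_ear bones, widest at the last segment.
# Angular constraints link each segment to the one before it, rooted at the pack, with matching swing limits and very tight twist so the ears bend with little twisting.
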PhysicsAsset /BRCosmetics/Accessories/FORT_Backpacks/F_MED_Bani_Backpack/Meshes/F_MED_Bani_Pack_Physics
bodies {
 dyn_ear_l_01 [{shapeType=Capsule, length=5.093, radius=1.107, center=<2.8 0 0>, rotation=<rot 90 0 0>}]
 dyn_ear_l_02 [{shapeType=Capsule, length=6.6869, radius=1.313, center=<4 0 0>, rotation=<rot 90 0 0>}]
 dyn_ear_l_03 [{shapeType=Capsule, length=5.3634, radius=2.6366, center=<4 0 0>, rotation=<rot 90 0 0>}]
 dyn_ear_r_01 [{shapeType=Capsule, length=5.093, radius=1.107, center=<2.8 0 0>, rotation=<rot 90 0 0>}]
 dyn_ear_r_02 [{shapeType=Capsule, length=6.6869, radius=1.313, center=<4 0 0>, rotation=<rot 90 0 0>}]
 dyn_ear_r_03 [{shapeType=Capsule, length=5.3634, radius=2.6366, center=<4 0 0>, rotation=<rot 90 0 0>}]
 pack [{shapeType=Capsule, length=12.2067, radius=9.2037, center=<0 -6.0495 -5.8454>}]
}
constraints [
 {bone1Name=dyn_ear_l_01, bone2Name=pack, swing1LimitDegrees=15, swing2LimitDegrees=35, twistLimitDegrees=2.5}
 {bone1Name=dyn_ear_l_02, bone2Name=dyn_ear_l_01, swing1LimitDegrees=15, swing2LimitDegrees=35, twistLimitDegrees=2.5}
 {bone1Name=dyn_ear_l_03, bone2Name=dyn_ear_l_02, swing1LimitDegrees=15, swing2LimitDegrees=35, twistLimitDegrees=2.5}
 {bone1Name=dyn_ear_r_01, bone2Name=pack, swing1LimitDegrees=15, swing2LimitDegrees=35, twistLimitDegrees=2.5}
 {bone1Name=dyn_ear_r_02, bone2Name=dyn_ear_r_01, swing1LimitDegrees=15, swing2LimitDegrees=35, twistLimitDegrees=2.5}
 {bone1Name=dyn_ear_r_03, bone2Name=dyn_ear_r_02, swing1LimitDegrees=15, swing2LimitDegrees=35, twistLimitDegrees=2.5}
]
