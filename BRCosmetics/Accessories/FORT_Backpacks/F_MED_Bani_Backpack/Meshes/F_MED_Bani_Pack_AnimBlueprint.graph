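# Animation Blueprint for the Bani backpack cosmetic that adds physics-driven secondary motion to the pack's skeleton.
# The live chain takes the reference pose, nudges the pack bone along its local Z, then runs a RigidBody node with the pack's physics asset and overridden gravity.
# An AnimDynamics right-ear chain, copied by rotation delta to the left ear, is disconnected from the output.
# The event graph has no logic.
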AnimBlueprint /BRCosmetics/Accessories/FORT_Backpacks/F_MED_Bani_Backpack/Meshes/F_MED_Bani_Pack_AnimBlueprint
target_skeleton=&/F_MED_Bani_Pack_Skeleton
graphs [
 AnimGraph {nodes [
   n1 = Anim Dynamics - Bone: dyn_ear_r_01(Alpha=1, ExternalForce="0,0,0") {
    angularDampingOverride=0.9
    bGravityOverrideInSimSpace=true
    bOverrideAngularDamping=true
    bUseGravityOverride=true
    alphaBoolBlend {blendOption=Linear}
    boundBone {boneName=dyn_ear_r_01}
    componentAppliedLinearAccClamp=<1500 1500 1500>
    componentLinearAccScale=<1 0.2 0.2>
    componentLinearVelScale=<0.3 0.3 0.3>
    gravityOverride {z=-980}
    physicsBodyDefinitions [{
      sphereCollisionRadius=0
      boundBone {boneName=dyn_ear_r_01}
      boxExtents=<6 4 4>
      constraintSetup {angularLimitsMax=<3 3 0>, angularLimitsMin=<-3 -35 0>, angularTarget=<0 0 0>}
      localJointOffset=<0.5 0 0>
     }]
   }
   n2 = Anim Dynamics - Bone: dyn_ear_r_02(ComponentPose=@n1, Alpha=1, ExternalForce="0,0,0") {
    angularDampingOverride=0.9
    bGravityOverrideInSimSpace=true
    bOverrideAngularDamping=true
    bUseGravityOverride=true
    alphaBoolBlend {blendOption=Linear}
    boundBone {boneName=dyn_ear_r_02}
    componentAppliedLinearAccClamp=<1500 1500 1500>
    componentLinearAccScale=<1 0.2 0.2>
    componentLinearVelScale=<0.3 0.3 0.3>
    gravityOverride {z=-980}
    physicsBodyDefinitions [{
      sphereCollisionRadius=0
      boundBone {boneName=dyn_ear_r_02}
      boxExtents=<6 4 4>
      constraintSetup {angularLimitsMax=<3 3 0>, angularLimitsMin=<-3 -35 0>, angularTarget=<0 0 0>}
      localJointOffset=<0.5 0 0>
     }]
   }
   n3 = Anim Dynamics - Bone: dyn_ear_r_03(ComponentPose=@n2, Alpha=1, ExternalForce="0,0,0") {
    angularDampingOverride=0.9
    bGravityOverrideInSimSpace=true
    bOverrideAngularDamping=true
    bUseGravityOverride=true
    alphaBoolBlend {blendOption=Linear}
    boundBone {boneName=dyn_ear_r_03}
    componentAppliedLinearAccClamp=<1500 1500 1500>
    componentLinearAccScale=<1 0.2 0.2>
    componentLinearVelScale=<0.3 0.3 0.3>
    gravityOverride {z=-980}
    physicsBodyDefinitions [{
      sphereCollisionRadius=0
      boundBone {boneName=dyn_ear_r_03}
      boxExtents=<6 4 4>
      constraintSetup {angularLimitsMax=<3 3 0>, angularLimitsMin=<-3 -35 0>, angularTarget=<0 0 0>}
      localJointOffset=<0.5 0 0>
     }]
   }
   Copy Bone Delta - Source Bone: dyn_ear_r_03 - Target Bone: dyn_ear_l_03!(
    ComponentPose:Component Space Pose Link Structure=Copy Bone Delta - Source Bone: dyn_ear_r_02 - Target Bone: dyn_ear_l_02!(
     ComponentPose:Component Space Pose Link Structure=Copy Bone Delta - Source Bone: dyn_ear_r_01 - Target Bone: dyn_ear_l_01!(
      ComponentPose:Component Space Pose Link Structure=@n3
      Alpha:float=1
      bCopyTranslation:bool=False
      bCopyRotation:bool=true
      bCopyScale:bool=False
     ) -> (Pose:Component Space Pose Link Structure) {
      is=title AnimGraphNode_CopyBoneDelta
      alphaBoolBlend {blendOption=Linear}
      sourceBone {boneName=dyn_ear_r_01}
      targetBone {boneName=dyn_ear_l_01}
     }
     Alpha:float=1
     bCopyTranslation:bool=False
     bCopyRotation:bool=true
     bCopyScale:bool=False
    ) -> (Pose:Component Space Pose Link Structure) {
     is=title AnimGraphNode_CopyBoneDelta
     alphaBoolBlend {blendOption=Linear}
     sourceBone {boneName=dyn_ear_r_02}
     targetBone {boneName=dyn_ear_l_02}
    }
    Alpha:float=1
    bCopyTranslation:bool=False
    bCopyRotation:bool=true
    bCopyScale:bool=False
   ) -> (Pose:Component Space Pose Link Structure) {
    is=title AnimGraphNode_CopyBoneDelta
    alphaBoolBlend {blendOption=Linear}
    sourceBone {boneName=dyn_ear_r_03}
    targetBone {boneName=dyn_ear_l_03}
   }
   n4 = "Transform (Modify) Bone - Bone: pack"(ComponentPose=Local To Component(LocalPose=Local Space Ref Pose()), Translation="0.000,0.000,2.5", Rotation="0,0,0") {
    translationMode=BMM_Additive
    translationSpace=BCS_BoneSpace
    alphaBoolBlend {blendOption=Linear}
    boneToModify {boneName=pack}
    translation {z=2.5}
   }
   n5 = RigidBody(ComponentPose=@n4) {
    bClampLinearTranslationLimitToRefPose=true
    bOverrideWorldGravity=true
    overridePhysicsAsset=&/F_MED_Bani_Pack_Physics
    alphaBoolBlend {blendOption=Linear}
    componentAppliedLinearAccClamp=<1500 1500 1500>
    componentLinearAccScale=<0.9 0.6 0.6>
    componentLinearVelScale=<1 0.8 0.8>
    overrideWorldGravity {z=-980}
   }
   Output Pose(Result=Component To Local(ComponentPose=@n5)) {layerGroup=None}
  ]}
 EventGraph {nodes [
   TryGetPawnOwner()
   Event BlueprintUpdateAnimation()
  ]}
]
variables []
